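# Sound cue for the Bee Husk's hive loop, an angry bee swarm sound.
# A Random node picks one of four equally weighted, looping mono bee recordings and feeds it through a Modulator that randomizes pitch, with a raised upper limit.
# An Attenuation node applies the close-range enemy beehive falloff.
# A concurrency override caps simultaneous loops at a few, stopping the farthest first, and the cue uses the ranged-weapon projectile effects sound class.
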SoundCue /Game/Sounds/Fort_Monster_Husk/Husk_Bee_Efforts/Husk_Bee_HiveLoop_Cue
root=Attenuation_3
nodes [
 {id=Attenuation_3, children [Modulator_1], properties {attenuationSettings=/Game/Sounds/Attenuation/Fort_Beehive_Enemy_Close_Att}}
 {id=Modulator_1, children [Random_0], properties {pitchMax=2}}
 {id=Random_0, children [WavePlayer_0,WavePlayer_1,WavePlayer_2,WavePlayer_3], properties {weights [1,1,1,1]}}
 {id=WavePlayer_0, properties {bLooping=true, soundWaveAssetPtr=/Game/Sounds/Fort_Monster_Husk/SFX/Insect_Bees_Angry_mono_01}}
 {id=WavePlayer_1, properties {bLooping=true, soundWaveAssetPtr=/Game/Sounds/Fort_Monster_Husk/SFX/Insect_Bees_Angry_mono_02}}
 {id=WavePlayer_2, properties {bLooping=true, soundWaveAssetPtr=/Game/Sounds/Fort_Monster_Husk/SFX/Insect_Bees_Angry_mono_03}}
 {id=WavePlayer_3, properties {bLooping=true, soundWaveAssetPtr=/Game/Sounds/Fort_Monster_Husk/SFX/Insect_Bees_Angry_mono_04}}
]
properties {
 bOverrideConcurrency=true
 maxDistance=1000
 soundClassObject=&Fort_WeaponRanged_Projectile_Effects
 volumeMultiplier=1
 attenuationOverrides {distanceAlgorithm=NaturalSound, falloffDistance=4000, attenuationShapeExtents {x=1000}}
 concurrencyOverrides {maxCount=3, resolutionRule=StopFarthestThenPreventNew}
}
